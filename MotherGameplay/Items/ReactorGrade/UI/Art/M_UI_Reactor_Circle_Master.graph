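# Translucent UI master material for the Reactor item grade that procedurally draws tintable circular rings, arcs, dots and progress bars.
# Two exponential radial gradients cut the ring, or supply a solid dot via static switch.
# Optional switches multiply in angular masks: a section driven by fill percentage with range correction, sine-based segment gaps and an alpha cut.
# Parameters also provide global rotation, fill-direction flip and overall alpha.
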Material /MotherGameplay/Items/ReactorGrade/UI/Art/M_UI_Reactor_Circle_Master
graph {
 properties {blendMode=BLEND_Translucent, materialDomain=MD_UI, numCustomizedUVs=1, pixelDepthOffsetMode=PDOM_Legacy, usedShadingModels=Default Lit}
 nodes [
  n1 = MaterialFunctionCall!(Rotation_Fraction=ScalarParameter() {group=Global, parameterName=Final_Rotation}, TexCoord=TextureCoordinate()) {materialFunction=/FortUILibrary/Foundation/Art_ToolKit/AngledBoxes/MF_UI_Rotator_Fractions}
  n2 = MaterialFunctionCall(Float2=@n1.Result) {materialFunction=&BreakOutFloat2Components}
  LinearInterpolate(A=@n1.Result, B=AppendVector(A=OneMinus(None=@n2.R), B=@n2.G), Alpha=ScalarParameter() {parameterName=FlipFillDirection, sliderMax=100})
  n3 = Saturate(None=MaterialFunctionCall(
    UVs=TextureCoordinate()
    Radius=ScalarParameter() {defaultValue=0.25, group=Properties, parameterName=Radius}
    Density=ScalarParameter() {defaultValue=16, group=EdgeHardness, parameterName=Density_Outer}
   ) {materialFunction=&RadialGradientExponential}.RadialGradientExponential)
  n4 = Subtract(
   A=MaterialFunctionCall(
    Rotation_Fraction=ScalarParameter() {defaultValue=0.25, group=AlphaMask, parameterName=Section_Orientation_AlphaMask}
    TexCoord=TextureCoordinate()
   ) {materialFunction=&MF_UI_Rotator_Fractions}.Result
   B=Constant() {r=0.5}
  )
  n5 = Saturate(None=MaterialFunctionCall(Vector or UVs=Multiply(A=@n4) {constB=2}) {materialFunction=&VectorToRadialValue}.Vector Converted to Angle)
  n6 = Multiply(A=ScalarParameter() {defaultValue=25, group=AlphaMask, parameterName=Section_Percentage_AlphaMask}) {constB=0.01}
  n7 = Saturate(None=Multiply(A=Subtract(A=@n5, B=@n6) {constB=0.5}, B=ScalarParameter() {defaultValue=3, group=AlphaMask, parameterName=AlphaMask_Lift}) {constB=512})
  n8 = ScalarParameter() {defaultValue=0.25, group=Properties, parameterName=Radius}
  n9 = Multiply(A=ScalarParameter() {defaultValue=8, group=Properties, parameterName=RingWidth}) {constB=0.01}
  n10 = Saturate(None=MaterialFunctionCall(
    UVs=TextureCoordinate()
    Radius=Subtract(A=@n8, B=@n9)
    Density=ScalarParameter() {defaultValue=12, group=EdgeHardness, parameterName=Density_Inner}
   ) {materialFunction=&RadialGradientExponential}.RadialGradientExponential)
  n11 = Saturate(None=DotProduct(A=@n3, B=OneMinus(None=@n10)))
  n12 = MaterialFunctionCall(Rotation_Fraction=ScalarParameter() {defaultValue=0.25, group=Section, parameterName=Section_Orientation}, TexCoord=TextureCoordinate()) {materialFunction=&MF_UI_Rotator_Fractions}
  n13 = Constant() {r=0.5}
  n14 = MaterialFunctionCall(Vector or UVs=Multiply(A=Subtract(A=@n12.Result, B=@n13)) {constB=2}) {materialFunction=&VectorToRadialValue}
  n15 = Multiply(A=ScalarParameter() {defaultValue=25, group=Section, parameterName=Section_Percentage}) {constB=0.01}
  n16 = Multiply(A=ScalarParameter() {defaultValue=100, parameterName=FillPercent, sliderMax=100}) {constB=0.01}
  n17 = ScalarParameter() {parameterName=Correction_Start}
  n18 = ScalarParameter() {defaultValue=1, parameterName=Correction_End}
  n19 = LinearInterpolate(
   B=@n15
   Alpha=LinearInterpolate(A=@n16, B=LinearInterpolate(A=@n17, B=@n18, Alpha=@n16), Alpha=ScalarParameter() {parameterName=HasPercentCorrection})
  )
  n20 = Multiply(A=Subtract(A=Saturate(None=@n14.Vector Converted to Angle), B=StaticSwitchParameter(True=@n19, False=@n15) {parameterName=IsProgressBar}) {constB=0.5}) {constB=512}
  n21 = Saturate(None=OneMinus(None=Saturate(None=@n20)))
  n22 = StaticSwitchParameter(True=Saturate(None=DotProduct(A=@n11, B=@n21)), False=@n11) {parameterName=IsSection}
  n23 = MaterialFunctionCall() {materialFunction=&VectorToRadialValue}
  n24 = Floor(None=ScalarParameter() {defaultValue=12, group=Segments, parameterName=Segments})
  n25 = ComponentMask(None=Sine(None=Multiply(A=@n23.Vector Converted to Angle, B=@n24)) {period=0.5}) {r=true}
  n26 = Multiply(A=ScalarParameter() {defaultValue=8, group=Segments, parameterName=GapSize}) {constB=0.1}
  n27 = Saturate(None=Multiply(A=Subtract(A=@n25, B=@n26), B=ScalarParameter() {defaultValue=8, group=Segments, parameterName=SegmentSoftness}) {constB=8})
  n28 = StaticSwitchParameter(True=Saturate(None=DotProduct(A=@n22, B=@n27)), False=@n22) {parameterName=HasSegments}
  n29 = StaticSwitchParameter(True=Saturate(None=DotProduct(A=@n7, B=@n28)), False=@n28) {parameterName=HasAlphaMask}
  Saturate(None=Multiply(
    A=StaticSwitchParameter(True=@n3, False=@n29) {parameterName=IsDot}
    B=ScalarParameter() {defaultValue=1, group=Global, parameterName=Overall_Alpha}
   ))
  VectorParameter() {group=Color, parameterName=Color, defaultValue=<rgba 1 0 0 1>}
 ]
}
