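# Translucent UI master material that procedurally draws a piece of the Reactor sights, with two-color tinting and global alpha controls.
# Rotated, offset UV ramps with adjustable softness form the angled, top, bottom and cut edges, avoiding textures.
# The IsFIll static switch selects either a radial-gradient fill clipped by a horizontal offset mask, or an outline built from inset edges.
# An optional sine-based center mask cuts into the opacity.
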Material /MotherGameplay/Items/ReactorGrade/UI/Art/M_UI_Reactor_Sights_Component_Master
graph {
 properties {blendMode=BLEND_Translucent, materialDomain=MD_UI, pixelDepthOffsetMode=PDOM_Legacy, usedShadingModels=Default Lit}
 nodes [
  n1 = TextureCoordinate()
  n2 = ComponentMask(None=Add(A=@n1, B=ScalarParameter() {defaultValue=-0.5, parameterName=FillMask_Offset}) {constB=-0.5}) {r=true}
  n3 = Saturate(None=OneMinus(None=Ceil(None=@n2)))
  n4 = Constant() {r=0.5}
  n5 = Add(A=@n4, B=ScalarParameter() {parameterName=Fill_XOffset})
  n6 = Add(A=@n4, B=ScalarParameter() {parameterName=Fill_YOffset})
  n7 = OneMinus(None=MaterialFunctionCall(
    UVs=@n1
    CenterPosition=AppendVector(A=@n5, B=@n6)
    Radius=ScalarParameter() {defaultValue=0.333, parameterName=FillRadius}
    Density=ScalarParameter() {defaultValue=2, parameterName=FillDensity}
   ) {materialFunction=&RadialGradientExponential}.RadialGradientExponential)
  n8 = ScalarParameter() {defaultValue=0.05, group=Angle_1, parameterName=Angle_1_Rotation}
  n9 = TextureCoordinate()
  n10 = ScalarParameter() {defaultValue=-0.325, group=Angle_1, parameterName=Angle_1_Offset}
  n11 = ComponentMask(None=Add(
    A=MaterialFunctionCall!(Rotation_Fraction=@n8, TexCoord=@n9) {materialFunction=/FortUILibrary/Foundation/Art_ToolKit/AngledBoxes/MF_UI_Rotator_Fractions}.Result
    B=@n10
   )) {r=true}
  n12 = ScalarParameter() {defaultValue=32, group=Softness, parameterName=Angle_1_Softness}
  n13 = MaterialFunctionCall(Rotation_Fraction=ScalarParameter() {defaultValue=-0.05, group=TopCut, parameterName=Top_Cut_1_Rotation}, TexCoord=TextureCoordinate()) {materialFunction=&MF_UI_Rotator_Fractions}
  n14 = ScalarParameter() {defaultValue=-0.125, group=TopCut, parameterName=Top_Cut_1_Offset}
  n15 = Multiply(
   A=ComponentMask(None=Add(A=@n13.Result, B=@n14)) {g=true}
   B=ScalarParameter() {defaultValue=32, group=Softness, parameterName=Top_Cut_1_Softness}
  )
  n16 = ScalarParameter() {group=Top, parameterName=Top_Rotation}
  n17 = TextureCoordinate()
  n18 = MaterialFunctionCall(Rotation_Fraction=@n16, TexCoord=@n17) {materialFunction=&MF_UI_Rotator_Fractions}
  n19 = ScalarParameter() {defaultValue=-0.125, group=Top, parameterName=Top_Offset}
  n20 = ScalarParameter() {defaultValue=32, group=Softness, parameterName=Top_Softness}
  n21 = Multiply(A=ComponentMask(None=Add(A=@n18.Result, B=@n19)) {g=true}, B=@n20)
  n22 = DotProduct(A=Saturate(None=Multiply(A=@n11, B=@n12)), B=LinearInterpolate(B=Saturate(None=@n15), Alpha=Saturate(None=@n21)))
  n23 = ScalarParameter() {group=Bottom, parameterName=Bottom_Rotation}
  n24 = TextureCoordinate()
  n25 = ScalarParameter() {defaultValue=-0.75, group=Bottom, parameterName=Bottom_Offset}
  n26 = Add(
   A=MaterialFunctionCall(Rotation_Fraction=@n23, TexCoord=@n24) {materialFunction=&MF_UI_Rotator_Fractions}.Result
   B=@n25
  )
  n27 = ScalarParameter() {defaultValue=32, group=Softness, parameterName=Bottom_Softness}
  n28 = Saturate(None=Multiply(A=ComponentMask(None=@n26) {g=true}, B=@n27))
  n29 = Saturate(None=DotProduct(A=Saturate(None=@n22), B=OneMinus(None=@n28)))
  n30 = Saturate(None=DotProduct(A=Saturate(None=@n7), B=@n29))
  n31 = Add(
   A=OneMinus(None=MaterialFunctionCall(Rotation_Fraction=@n8, TexCoord=@n9) {materialFunction=&MF_UI_Rotator_Fractions}.Result)
   B=Subtract(A=@n10, B=ScalarParameter() {defaultValue=0.25, group=Angle_1, parameterName=Angle_1_Width})
  )
  n32 = Saturate(None=Multiply(A=ComponentMask(None=@n31) {r=true}, B=@n12))
  n33 = Add(
   A=MaterialFunctionCall(Rotation_Fraction=ScalarParameter() {defaultValue=-0.0375, group=TopCut, parameterName=Top_Cut_2_Rotation}, TexCoord=TextureCoordinate()) {materialFunction=&MF_UI_Rotator_Fractions}.Result
   B=ScalarParameter() {defaultValue=-0.175, group=TopCut, parameterName=Top_Cut_2_Offset}
  )
  n34 = Saturate(None=Multiply(A=ComponentMask(None=@n33) {g=true}, B=ScalarParameter() {defaultValue=32, group=Softness, parameterName=Top_Cut_2_Softness}))
  n35 = MaterialFunctionCall(Rotation_Fraction=@n16, TexCoord=@n17) {materialFunction=&MF_UI_Rotator_Fractions}
  n36 = ScalarParameter() {defaultValue=0.6375, group=Top, parameterName=TopOffset_Width}
  n37 = ComponentMask(None=Add(A=OneMinus(None=@n35.Result), B=Subtract(A=@n19, B=@n36))) {g=true}
  n38 = OneMinus(None=Saturate(None=Multiply(A=@n37, B=@n20)))
  n39 = Saturate(None=DotProduct(A=OneMinus(None=@n32), B=LinearInterpolate(B=@n34, Alpha=@n38)))
  n40 = OneMinus(None=MaterialFunctionCall(Rotation_Fraction=ScalarParameter() {defaultValue=-0.0125, group=BottomCut, parameterName=Bottom_Cut_Rotation}, TexCoord=TextureCoordinate()) {materialFunction=&MF_UI_Rotator_Fractions}.Result)
  n41 = ScalarParameter() {defaultValue=-0.25, group=BottomCut, parameterName=Bottom_Cut_Offset}
  n42 = Multiply(A=ComponentMask(None=Add(A=@n40, B=@n41)) {g=true}, B=ScalarParameter() {defaultValue=32, group=Softness, parameterName=Bottom_Cut_Softness})
  n43 = OneMinus(None=MaterialFunctionCall(Rotation_Fraction=@n23, TexCoord=@n24) {materialFunction=&MF_UI_Rotator_Fractions}.Result)
  n44 = Subtract(A=@n25, B=ScalarParameter() {defaultValue=-0.5, group=Bottom, parameterName=BottomOffset_Width})
  n45 = Multiply(A=ComponentMask(None=Add(A=@n43, B=@n44)) {g=true}, B=@n27)
  n46 = Saturate(None=LinearInterpolate(A=Saturate(None=@n42), Alpha=Saturate(None=@n45)))
  n47 = OneMinus(None=Saturate(None=DotProduct(A=@n39, B=@n46)))
  n48 = StaticSwitchParameter(True=Saturate(None=DotProduct(A=@n3, B=@n30)), False=Saturate(None=LinearInterpolate(B=@n47, Alpha=@n29))) {parameterName=IsFIll}
  LinearInterpolate(
   A=VectorParameter() {group=Color, parameterName=Color_2, defaultValue=<rgba 1 0 0 1>}.RGB
   B=VectorParameter() {group=Color, parameterName=Color_1, defaultValue=<rgba 0 1 0 1>}.RGB
   Alpha=@n48
  )
  n49 = TextureCoordinate()
  n50 = ScalarParameter() {parameterName=ShiftCenterMask}
  n51 = Sine(None=Multiply(A=Add(A=@n49, B=@n50), B=ScalarParameter() {defaultValue=1, parameterName=ScaleCenterMask})) {period=2}
  n52 = Saturate(None=Power(Base=ComponentMask(None=@n51) {g=true}, Exp=ScalarParameter() {defaultValue=1, parameterName=TightenCenterMask}))
  OneMinus(None=@n52)
  n53 = ScalarParameter() {defaultValue=1, parameterName=CenterMaskEdgeHardness}
  n54 = OneMinus(None=Saturate(None=Multiply(A=@n52, B=@n53)))
  n55 = Multiply(
   A=LinearInterpolate(A=@n48, B=LinearInterpolate(B=@n48, Alpha=@n54), Alpha=ScalarParameter() {parameterName=UseCenterMask})
   B=ScalarParameter() {defaultValue=1, group=Global, parameterName=FinalAlpha}
  )
  Saturate(None=Multiply(A=Saturate(None=@n55), B=ScalarParameter() {defaultValue=1, group=Global, parameterName=AlphaCompensation}))
 ]
}
